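; Actor for the Sneaky Snowman consumable that stands in for the snowman on a player, kept hidden in game.
; On begin play it stores the instigator as the player pawn, disables mesh collision, and destroys itself after a scalable-float duration.
; Server damage events accumulate into a total.
; When that total reaches max health, it disables collision, sends an ability-activate gameplay event to the instigator and destroys itself.
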(event EventBeginPlay
  ;[1 1 1] Make sure this snowman is not visible
  (Rendering|SetActorHiddenInGame true)
  (bind _asfort_player_pawn (Utilities|Casting|CastToFortPlayerPawn (Variables|Actor|GetInstigator)))
  (Variables|Default|SetPlayerPawn _asfort_player_pawn)
  (Collision|SetCollisionEnabled (Variables|Debug|GetStaticMeshComponent) "NoCollision")
  (Utilities|FlowControl|Delay (ScalableFloat|GetValueatLevel (Variables|Default|GetSnowmanDuration)))
  ;[1 1 1] When this snowman on the player is hit, destroy the actor.
  (Actor|DestroyActor))

(event EventOnDamageServer
  (params (Damage :type Float%20%28single-precision%29) (DamageTags :type Gameplay%20Tag%20Container%20Structure%20%28by%20ref%29) (Momentum :type Vector :default "0, 0, 0") (HitInfo :type Hit%20Result%20Structure%20%28by%20ref%29) (InstigatedBy :type Controller%20Object%20Reference) (DamageCauser :type Actor%20Object%20Reference) (EffectContext :type Gameplay%20Effect%20Context%20Handle%20Structure))
  ;[1 1 1] When this snowman on the player is hit, destroy the actor.
  (bind _output_get (Variables|Default|SetTotalDamage (+ Damage (Variables|Default|GetTotalDamage))))
  (if (>= _output_get (ScalableFloat|GetValueatLevel (Variables|Default|GetMaxHealth)))
    (Collision|SetCollisionEnabled (Variables|Debug|GetStaticMeshComponent) "NoCollision")
    (Ability|SendGameplayEventtoActor (Variables|Actor|GetInstigator) "(TagName=\"Abilities.Generic.ActiveAbility.Activate\")")
    (Actor|DestroyActor)))

(fn ConstructionScript)
