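; Sneaky Snowman disguise actor that acts as a hidden, damageable proxy for the instigating player pawn and shares its team.
; On BeginPlay it caches the current weapon's item GUID and starts a repeating timer that enables query-only cylinder collision while the pawn crouches.
; Non-teammate damage accumulates until MaxHealth, then sends an ability-activate event to the instigator and destroys the actor.
; On destruction it removes that item from inventory if its stack count is zero.
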(event Custom|RemoveGUIDFromInvWithZeroCount
  (params (Item%20GUID :type Guid%20Structure))
  ;[0.9167 0.9167 0.9167 0.5] Remove item with given GUID from player inventory if count is 0
  (bind _playercontroller (Variables|Default|GetPlayerController)
    (:then
      (bind _returnvalue (Inventory|GetItemNuminStackbyGuid _playercontroller Item%20GUID))
      (if (== _returnvalue 0)
        (bind _returnvalue_2 (Inventory|GetInventoryItemwithGuid _playercontroller Item%20GUID))
        (Inventory|RemoveFortItemFromPlayer _playercontroller _returnvalue_2 -1 true)))
    (:else)))

(event Custom|InitializeSnowman
  ;[1 1 1 0.5] Assign variables for disguised pawn
  (Variables|Default|SetPlayerController (Fortnite|GetFortPlayerControllerfromActor (Variables|Default|GetPlayerPawn)))
  (Variables|Default|SetItemGUID (FortWeapon|GetInventoryGUID (Weapon|GetCurrentWeapon (Variables|Default|GetPlayerPawn)))))

(event EventDestroyed
  (CallFunction|RemoveGUIDFromInvwithZeroCount (Variables|Default|GetItemGUID)))

(event EventBeginPlay
  (bind _returnvalue (Game|GetInstigator))
  ;[1 1 1] EVENT -> Begin Play
  (Rendering|SetActorHiddenInGame true)
  (Collision|SetCollisionEnabled (Variables|Default|GetCylinder) "NoCollision")
  (bind _asfort_player_pawn (Utilities|Casting|CastToFortPlayerPawn _returnvalue))
  (Variables|Default|SetPlayerPawn _asfort_player_pawn)
  (CallFunction|InitializeSnowman)
  (Utilities|Time|SetTimerbyEvent (EventDispatchers|CreateEvent :!event Custom|CrouchedCheck) 0.2 true)
  (if (Utilities|IsValid _returnvalue :!class /Script/Engine.KismetSystemLibrary)
    (CallFunction|SetTeam (Team|GetActorTeam _returnvalue))))

(event EventOnDamageServer
  (params (Damage :type Float%20%28single-precision%29) (DamageTags :type Gameplay%20Tag%20Container%20Structure%20%28by%20ref%29) (Momentum :type Vector :default "0, 0, 0") (HitInfo :type Hit%20Result%20Structure%20%28by%20ref%29) (InstigatedBy :type Controller%20Object%20Reference) (DamageCauser :type Actor%20Object%20Reference) (EffectContext :type Gameplay%20Effect%20Context%20Handle%20Structure))
  ;[1 1 1] EVENT -> On Damage - Destroy actor if enough damage has been taken
  (if (not (Team|OnSameTeam DamageCauser (Variables|Actor|GetInstigator)))
    (bind _output_get (Variables|Default|SetTotalDamage (+ Damage (Variables|Default|GetTotalDamage))))
    (if (>= _output_get (ScalableFloat|GetValueatLevel (Variables|Default|GetMaxHealth)))
      (Ability|SendGameplayEventtoActor (Variables|Actor|GetInstigator) "(TagName=\"Abilities.Generic.ActiveAbility.Activate\")")
      (Actor|DestroyActor))))

(event EventOnDeathServer
  (params (Damage :type Float%20%28single-precision%29) (DamageTags :type Gameplay%20Tag%20Container%20Structure%20%28by%20ref%29) (Momentum :type Vector :default "0, 0, 0") (HitInfo :type Hit%20Result%20Structure%20%28by%20ref%29) (InstigatedBy :type Controller%20Object%20Reference) (DamageCauser :type Actor%20Object%20Reference) (EffectContext :type Gameplay%20Effect%20Context%20Handle%20Structure))
  ;[1 1 1] EVENT -> On Death
  (Collision|SetCollisionEnabled (Variables|Default|GetSM_SneakySnowman) "NoCollision"))

(event Custom|CrouchedCheck
  ;[1 1 1] Crouched Check
  (if (Variables|/Script/Engine.Character|GetbIsCrouched (Variables|Default|GetPlayerPawn))
    (Collision|SetCollisionEnabled (Variables|Default|GetCylinder) "QueryOnly")
    (else
      (Collision|SetCollisionEnabled (Variables|Default|GetCylinder) "NoCollision"))))

(fn ConstructionScript
  (CallFunction|SetTeam (Team|GetActorTeam (Game|GetInstigator))))
