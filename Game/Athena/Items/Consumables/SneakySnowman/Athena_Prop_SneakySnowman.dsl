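; Sneaky Snowman prop Blueprint that falls via projectile movement and can apply a snowman effect to players or vehicle passengers it hits.
; Movement waits for playlist load, then downward traces and building death events restart falling when support is lost.
; Landing destroys it after excessive falls, player or vehicle hits; pickups are ignored.
; Replicated landed location resyncs clients, and material parameters drive spawn and damage wobbles.
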(event EventBeginPlay
  (bind (_start _end) (CallFunction|TraceStartandEnd))
  (bind (_asfort_player_pawn _bsuccess) (Utilities|Casting|CastToFortPlayerPawn (Game|GetInstigator)))
  ;[1 1 1] EVENT -> BeginPlay: Double check if gamestate has not loaded then we will turn off projectile movement. This is to fix Snowmen from falling through the world
  (Utilities|FlowControl|Sequence
    (:then_0
      (Networking|SwitchHasAuthority
        (:Authority
          (bind _asfort_game_state_athena (Utilities|Casting|CastToFortGameStateAthena (Game|GetGameState)))
          (if (Variables|/Script/FortniteGame.FortGameStateAthena|GetbPlaylistDataIsLoaded _asfort_game_state_athena)
            (Variables|Default|SetGamestateLoaded true)
            (else
              (Components|Activation|Deactivate (Variables|Default|GetProjectileMovementComp))
              (Playlist|BindEventtoOnCurrentPlaylistLoaded _asfort_game_state_athena (delegate Custom|PlaylistLoaded)))))))
    (:then_1
      (Utilities|Time|SetTimerbyEvent (EventDispatchers|CreateEvent :!event Custom|WobbleAnims) 0.1))
    (:then_2
      (if _bsuccess
        _
        (elif (Components|Activation|IsActive (Variables|Default|GetProjectileMovementComp))
          ;[1 1 1] Activate projectile movement if the props in world are sitting on a destructable object
          (Components|Activation|Deactivate (Variables|Default|GetProjectileMovementComp))))
      (bind _outhit (Collision|LineTraceForObjects _start _end (Variables|Default|GetObjectTypes) false (Utilities|Array|MakeArray self) "Persistent"))
      (bind (_outhit_bblockinghit _outhit_binitialoverlap _outhit_time _outhit_distance _outhit_location _outhit_impactpoint _outhit_normal _outhit_impactnormal _outhit_physmat _outhit_hitactor) (Collision|BreakHitResult _outhit))
      (bind _asbuilding_actor (Utilities|Casting|CastToBuildingActor _outhit_hitactor))
      (if (Variables|Actor|GetbCanBeDamaged _asbuilding_actor)
        (EventDispatchers|AssignOnDied _asbuilding_actor (delegate Custom|WorldPropSupportDied) :!class /Script/FortniteGame.BuildingActor)
        (Components|Activation|Activate (Variables|Default|GetProjectileMovementComp) true)))))

(event Custom|PlaylistLoaded
  (params (PlaylistName :type Name) (PlaylistContextTags :type Gameplay%20Tag%20Container%20Structure%20%28by%20ref%29))
  (bind (_start _end) (CallFunction|TraceStartandEnd))
  ;[1 1 1] Activate projectile movement if the props in world are sitting on a destructable object
  (if (Components|Activation|IsActive (Variables|Default|GetProjectileMovementComp))
    (Components|Activation|Deactivate (Variables|Default|GetProjectileMovementComp)))
  (bind _outhit (Collision|LineTraceForObjects _start _end (Variables|Default|GetObjectTypes) false (Utilities|Array|MakeArray self) "Persistent"))
  (bind (_outhit_bblockinghit _outhit_binitialoverlap _outhit_time _outhit_distance _outhit_location _outhit_impactpoint _outhit_normal _outhit_impactnormal _outhit_physmat _outhit_hitactor) (Collision|BreakHitResult _outhit))
  (bind _asbuilding_actor (Utilities|Casting|CastToBuildingActor _outhit_hitactor))
  (if (Variables|Actor|GetbCanBeDamaged _asbuilding_actor)
    (EventDispatchers|AssignOnDied _asbuilding_actor (delegate Custom|WorldPropSupportDied) :!class /Script/FortniteGame.BuildingActor)
    (Components|Activation|Activate (Variables|Default|GetProjectileMovementComp) true)))

(event Custom|WorldPropSupportDied
  (params (DamagedActor :type Actor%20Object%20Reference) (Damage :type Float%20%28single-precision%29) (InstigatedBy :type Controller%20Object%20Reference) (DamageCauser :type Actor%20Object%20Reference) (HitLocation :type Vector) (FHitComponent :type Primitive%20Component%20Object%20Reference) (BoneName :type Name) (Momentum :type Vector))
  ;[1 1 1] Activate projectile movement if the props in world are sitting on a destructable object
  (CallFunction|RestartSimulating))

(event EventPreFinishSpawning
  (params (VolumeSpawnedWithin :type Actor%20Object%20Reference))
  ;[1 1 1] Disable the snowman movement until the entire save is completed.
  (|CallParentFunction 0)
  (Components|Activation|Deactivate (Variables|Default|GetProjectileMovementComp)))

(event Custom|RestartSimulating
  ;[1 1 1] Restart Simulating: Restart Projectile Movement
  (Collision|IgnoreActorwhenMoving (Variables|Debug|GetStaticMeshComponent) (Variables|Default|GetPickupAthena) true)
  (Variables|Default|SetbSimulationRunning true)
  (Variables|Default|SetLandedLocation (Transformation|GetActorLocation))
  (CallFunction|ResetDistanceCheck))

(event EventOnDamagePlayEffects
  (params (Damage :type Float%20%28single-precision%29) (DamageTags :type Gameplay%20Tag%20Container%20Structure%20%28by%20ref%29) (Momentum :type Vector :default "0, 0, 0") (HitInfo :type Hit%20Result%20Structure%20%28by%20ref%29) (InstigatedBy :type Fort%20Pawn%20Object%20Reference) (DamageCauser :type Actor%20Object%20Reference) (EffectContext :type Gameplay%20Effect%20Context%20Handle%20Structure))
  ;[1 1 1] EVENT -> On Damage Play Effects: Trigger wobble anims on damage
  (bind _returnvalue (CallFunction|GetorCreateMIDOnDefaultMesh 0))
  (Rendering|Material|SetScalarParameterValue _returnvalue "WobbleStartTime" (Utilities|Time|GetGameTimeinSeconds) :!class /Script/Engine.MaterialInstanceDynamic)
  (Rendering|Material|SetScalarParameterValue _returnvalue "WobbleStrength" 0.25 :!class /Script/Engine.MaterialInstanceDynamic)
  (Rendering|Material|SetScalarParameterValue _returnvalue "WobbleDuration" 0.6 :!class /Script/Engine.MaterialInstanceDynamic))

(event OnComponentBeginOverlap(OverlapCheck)
  (params (OverlappedComponent :type Primitive%20Component%20Object%20Reference) (OtherActor :type Actor%20Object%20Reference) (OtherComp :type Primitive%20Component%20Object%20Reference) (OtherBodyIndex :type Integer) (bFromSweep :type Boolean) (SweepResult :type Hit%20Result%20Structure%20%28by%20ref%29))
  ;[1 1 1] EVENT -> OverlapCheck: Destroy if we are overlapping pawn on spawn
  (Utilities|Casting|CastToFortPlayerPawn OtherActor)
  (Actor|DestroyActor))

(event Custom|WobbleAnims
  ;[1 1 1] Trigger Wobble Anims on Spawn
  (if (Variables|Default|GetGamestateLoaded)
    (bind _returnvalue (CallFunction|GetorCreateMIDOnDefaultMesh 0))
    (Rendering|Material|SetScalarParameterValue _returnvalue "WobbleStartTime" (Utilities|Time|GetGameTimeinSeconds) :!class /Script/Engine.MaterialInstanceDynamic)
    (Rendering|Material|SetScalarParameterValue _returnvalue "WobbleStrength" 1.0 :!class /Script/Engine.MaterialInstanceDynamic)
    (Rendering|Material|SetScalarParameterValue _returnvalue "WobbleDuration" 1.0 :!class /Script/Engine.MaterialInstanceDynamic)
    ;[1 1 1] Force Set Location on Landing
    (Utilities|Time|SetTimerbyEvent (EventDispatchers|CreateEvent :!event ForceSetLocationOnLanding) 0.5)
    ;[1 1 1] Destroy Overlap component after delay because it's no longer needed
    (Utilities|FlowControl|Delay 0.5)
    (Components|DestroyComponent (Variables|Default|GetOverlapCheck))))

(event Custom|ResetDistanceCheck
  ;[1 1 1] Destroy ourself if we fell from too high
  (Utilities|FlowControl|DoOnce true
    (:Completed
      (bind _distance (Class|MissionBlueprintFunctionLibrary|DistanceBetweenTwoVectors (Variables|Default|GetLandedLocation) (.ImpactPoint (Collision|BreakHitResult (stale OnProjectileStop(ProjectileMovementComp) ImpactResult)))))
      (if (>= _distance (ScalableFloat|GetValueatLevel (Variables|Default|GetFallHeightToDestroy)))
        (Actor|DestroyActor)))))

(event OnProjectileStop(ProjectileMovementComp)
  (params (ImpactResult :type Hit%20Result%20Structure%20%28by%20ref%29))
  (bind (_bblockinghit _binitialoverlap _time _distance _location _impactpoint _normal _impactnormal _physmat _hitactor) (Collision|BreakHitResult ImpactResult))
  ;[1 1 1] EVENT -> On Stop
  (Variables|Default|SetbSimulationRunning false)
  (Utilities|FlowControl|Sequence
    (:then_0
      ;[1 1 1] Destroy ourself if we fell from too high
      (Utilities|FlowControl|DoOnce true
        (:Completed
          (bind _distance_2 (Class|MissionBlueprintFunctionLibrary|DistanceBetweenTwoVectors (Variables|Default|GetLandedLocation) _impactpoint))
          (if (>= _distance_2 (ScalableFloat|GetValueatLevel (Variables|Default|GetFallHeightToDestroy)))
            (Actor|DestroyActor)))))
    (:then_1
      (bind _asbuilding_actor (Utilities|Casting|CastToBuildingActor _hitactor)
        (:then
          (EventDispatchers|BindEventtoOnDied _asbuilding_actor (delegate Custom|BuildingActorDied) :!class /Script/FortniteGame.BuildingActor))
        (:CastFailed
          ;;[1 1 1] If we hit a player, apply snowman
          (bind _asfort_player_pawn (Utilities|Casting|CastToFortPlayerPawn _hitactor)
            (:then
              (if (and (Utilities|NotEqual(Object) _asfort_player_pawn (Variables|Actor|GetInstigator)) (> (ScalableFloat|GetValueatLevel (Variables|Default|GetShouldSnowmanEnemies)) 0.0))
                (Environment|ApplyGlobalEnvironmentGameplayEffecttoActor _asfort_player_pawn (Variables|Default|GetGE_ApplyEnemySnowman)))
              (Actor|DestroyActor))
            (:CastFailed
              ;;[1 1 1] If we hit a vehicle, apply snowman to driver and all passengers
              (bind _asfort_vehicle_interface (Utilities|Casting|CastToFortVehicleInterface _hitactor)
                (:then
                  (if (> (ScalableFloat|GetValueatLevel (Variables|Default|GetShouldSnowmanEnemies)) 0.0)
                    (if (> (Utilities|Array|Length (Vehicle|GetAllPassengers _asfort_vehicle_interface)) 0)
                      (for _array_element (Vehicle|GetAllPassengers _asfort_vehicle_interface)
                        (if (not (GameplayTags|HasMatchingGameplayTag _array_element "(TagName=\"Athena.Item.Snowman\")"))
                          (Environment|ApplyGlobalEnvironmentGameplayEffecttoActor _array_element (Variables|Default|GetGE_ApplyEnemySnowman))
                          (Actor|DestroyActor)))
                      (else
                        _)))
                  (Actor|DestroyActor))
                (:CastFailed
                  ;;[1 1 1] If we land on pickups then keep moving
                  (bind _asfort_pickup_athena (Utilities|Casting|CastToFortPickupAthena _hitactor))
                  (Variables|Default|SetPickupAthena _asfort_pickup_athena)
                  (CallFunction|RestartSimulating))))))))
    (:then_2
      ;;[1 1 1] Set Landed Location on server
      (Networking|SwitchHasAuthority
        (:Authority
          (Variables|Default|SetLandedLocation (Transformation|GetActorLocation)))))))

(event Custom|BuildingActorDied
  (params (DamagedActor :type Actor%20Object%20Reference) (Damage :type Float%20%28single-precision%29) (InstigatedBy :type Controller%20Object%20Reference) (DamageCauser :type Actor%20Object%20Reference) (HitLocation :type Vector) (FHitComponent :type Primitive%20Component%20Object%20Reference) (BoneName :type Name) (Momentum :type Vector))
  ;[1 1 1] EVENT -> On Stop
  (CallFunction|RestartSimulating))

(fn ForceSetLocationOnLanding
  (bind _projectilemovementcomp (Variables|Default|GetProjectileMovementComp))
  ;[1 1 1] When the snowman becomes net relevant again this resolved any disconnect between client and server
  (Networking|SwitchHasAuthority
    (:Remote
      (if (not (Math|Vector|Equal(Vector) (Variables|Default|GetLandedLocation) "0, 0, 0" 0.1))
        (Transformation|SetActorLocation (Variables|Default|GetLandedLocation) false true)
        (Variables|/Script/Engine.ProjectileMovementComponent|SetProjectileGravityScale 0.0 _projectilemovementcomp)
        (Components|Movement|StopMovementImmediately _projectilemovementcomp)))))

(fn OnRep_bSimulationRunning
  (if (Variables|Default|GetbSimulationRunning)
    (CallFunction|ResetSimulatedComponents)))

(fn OnRep_LandedLocation
  (Networking|SwitchHasAuthority
    (:Remote
      (Variables|/Script/Engine.ProjectileMovementComponent|SetProjectileGravityScale (Variables|Default|GetInitialGravityScale) (Variables|Default|GetProjectileMovementComp)))))

(fn ResetSimulatedComponents
  (Components|Movement|SetUpdatedComponent (Variables|Default|GetProjectileMovementComp) (Variables|Debug|GetStaticMeshComponent)))

(fn TraceStartAndEnd
  (return (Transformation|GetActorLocation) (- (Transformation|GetActorLocation) "0, 0,50.000000")))

(fn ConstructionScript
  (LOD|SetMaxDrawDistance (Variables|Debug|GetStaticMeshComponent) 0.0))
