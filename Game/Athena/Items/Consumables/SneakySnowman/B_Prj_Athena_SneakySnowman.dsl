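; Thrown Sneaky Snowman projectile that turns players it hits into snowmen or leaves a snowman prop where it lands.
; On touch, if enabled, the server applies a GameplayEffect to eligible pawns or vehicle passengers, excluding the thrower and existing snowmen.
; Shallow surfaces disable bouncing; stopping without a player hit spawns the prop and reports it to the actor limit.
; Explosion cosmetics use a server-set hit location; BeginPlay clears empty inventory stacks.
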(event EventBeginPlay
  ;[1 1 1] EVENT -> BeginPlay - If no more charges then remove item
  (bind _asfort_player_pawn (Utilities|Casting|CastToFortPlayerPawn (Game|GetInstigator)))
  (if (not (GameplayTags|HasMatchingGameplayTag _asfort_player_pawn "(TagName=\"Athena.Item.Snowman\")"))
    (bind _asfort_player_controller (Utilities|Casting|CastToFortPlayerController (Pawn|GetController _asfort_player_pawn)))
    (Networking|SwitchHasAuthority
      (:Authority
        ;[1 1 1] FORT-1020379; Cleaning up inventory slot properly. Logic below is old and invalid but left for prosperity's sake.
        (bind _returnvalue (Inventory|GetInventoryItemwithTag _asfort_player_controller (GameplayTags|MakeLiteralGameplayTag "(TagName=\"Item.Consumable.SneakySnowman\")") true))
        (if (== (Item|GetNuminStack _returnvalue) 0)
          (Inventory|RemoveFortItemFromPlayer _asfort_player_controller _returnvalue -1 true))))))

(event EventOnExploded
  (params (HitActors :type Array%20of%20Actor%20Object%20References) (HitResults :type Array%20of%20Hit%20Result%20Structures))
  (bind _hitlocation (Variables|SneakySnowman|GetHitLocation))
  ;[1 1 1] EVENT -> OnExploded - Audio and clean up
  (Networking|CanExecuteCosmeticEvents
    (:True
      (Effects|SpawnEmitteratLocation (Variables|FX|GetSnowmanSpawnFX) _hitlocation)
      (Camera|PlayWorldCameraShake "/Game/Athena/Items/Consumables/TowerGrenade/FX/B_Ground_CameraShake.B_Ground_CameraShake_C" _hitlocation 25.0 5e+01 2.0)
      (Audio|PlaySoundatLocation (Variables|Sounds|GetSoundOnExplo) _hitlocation)
      (Development|SetHiddeninGame (Variables|Default|GetSK_Sneaky_Snowman_Throw) true false)
      (Components|Activation|Deactivate (Variables|FX|GetFX_Trail))
      (Actor|DestroyActor))))

(event EventOnStop
  (params (Hit :type Hit%20Result%20Structure%20%28by%20ref%29))
  (bind (_returnvalue_roll _returnvalue_pitch _returnvalue_yaw) (Math|Rotator|BreakRotator (Transformation|GetActorRotation (Variables|Actor|GetInstigator))))
  (bind (_bblockinghit _binitialoverlap _time _distance _location _impactpoint _normal _impactnormal _physmat _hitactor) (Collision|BreakHitResult Hit))
  ;[1 1 1] EVENT -> On Stop do Hit FX and Spawn Snowman
  (if (not (Variables|SneakySnowman|GetHitPlayer))
    (CallFunction|SetHitLocation _location)
    (bind _asfort_player_pawn (Utilities|Casting|CastToFortPlayerPawn _hitactor)
      (:then
        (if (CallFunction|ShouldApplySnowman? _asfort_player_pawn)
          _))
      (:CastFailed
        (Networking|SwitchHasAuthority
          (:Authority
            (bind _returnvalue (Game|SpawnActorfromClass "/Game/Athena/Items/Consumables/SneakySnowman/Athena_Prop_SneakySnowman.Athena_Prop_SneakySnowman_C" (Math|Transform|MakeTransform _location (Math|Rotator|MakeRotator 0.0 0.0 (+ _returnvalue_yaw -9e+01)) "1.000000,1.000000,1.000000") "AlwaysSpawn" "MultiplyWithRoot" (Variables|Actor|GetInstigator) 0 0 (Variables|Actor|GetInstigator)))
            (Ability|SendGameplayEventtoActor (Game|GetInstigator) "(TagName=\"Abilities.Drops.AddToActorLimit.SneakySnowman\")" (Utilities|Struct|MakeGameplayEventData :Target _returnvalue :ContextHandle "()" :InstigatorTags "(GameplayTags=)" :TargetTags "(GameplayTags=)" :EventMagnitude 0.0 :TargetData "()"))))))
    (CallFunction|Explode)))

(event Custom|Explode
  ;[1 1 1] Explode (Custom Event)
  (Delivery|DoExplosion)
  ;;[1 1 1] Stop rotating movement
  (Components|Movement|SetUpdatedComponent (Variables|Default|GetRotatingMovement) 0))

(event Custom|SetHitLocation
  (params (HitLocation :type Vector))
  ;[1 1 1] FORT-1017579; So VFX always run at intended location
  (Networking|SwitchHasAuthority
    (:Authority
      (Variables|SneakySnowman|SetHitLocation HitLocation))))

(event EventOnTouched
  (params (OtherActor :type Actor%20Object%20Reference) (OtherComp :type Primitive%20Component%20Object%20Reference) (HitResult :type Hit%20Result%20Structure%20%28by%20ref%29) (bIsOverlap :type Boolean))
  (bind (_bblockinghit _binitialoverlap _time _distance _location _impactpoint _normal _impactnormal) (Collision|BreakHitResult HitResult))
  ;[1 1 1] EVENT -> OnTouched If we hit a player, apply Snowman to them
  (Variables|SneakySnowman|SetHitActor OtherActor)
  (CallFunction|SetHitLocation _location)
  (if (> (ScalableFloat|GetValueatLevel (Variables|SneakySnowman|GetShouldSnowmanEnemies)) 0.0)
    (bind _asfort_player_pawn (Utilities|Casting|CastToFortPlayerPawn (Variables|SneakySnowman|GetHitActor))
      (:then
        (Networking|SwitchHasAuthority
          (:Authority
            (Utilities|FlowControl|Sequence
              (:then_0
                (if (CallFunction|ShouldApplySnowman? _asfort_player_pawn)
                  (Environment|ApplyGlobalEnvironmentGameplayEffecttoActor _asfort_player_pawn (Variables|SneakySnowman|GetGE_ApplyEnemySnowman))
                  (Variables|SneakySnowman|SetHitPlayer true)))
              (:then_1
                (CallFunction|Explode))))))
      (:CastFailed
        (bind (_outactors _returnvalue) (Collision|BoxOverlapActors (Transformation|GetActorLocation) "48.000000,48.000000,200.000000" (Utilities|Array|MakeArray "ObjectTypeQuery3") "/Script/FortniteGame.FortPlayerPawnAthena" :!class /Script/Engine.KismetSystemLibrary))
        (if _returnvalue
          ;; FORT-1017579; Added for case where player hits themselves
          (Networking|SwitchHasAuthority
            (:Authority
              (Variables|SneakySnowman|SetHitPlayer true)
              (CallFunction|Explode)))
          (else
            ;;[1 1 1] If we hit a vehicle
            (bind _asfort_vehicle_interface (Utilities|Casting|CastToFortVehicleInterface (Variables|SneakySnowman|GetHitActor))
              (:then
                (if (> (Utilities|Array|Length (Vehicle|GetAllPassengers _asfort_vehicle_interface)) 0)
                  (Variables|SneakySnowman|SetHitPlayer true)
                  (Networking|SwitchHasAuthority
                    (:Authority
                      (for _array_element (Vehicle|GetAllPassengers _asfort_vehicle_interface)
                        (if (CallFunction|ShouldApplySnowman? _array_element)
                          (Environment|ApplyGlobalEnvironmentGameplayEffecttoActor _array_element (Variables|SneakySnowman|GetGE_ApplyEnemySnowman))))
                      (CallFunction|Explode)))))
              (:CastFailed
                _))))))
    (else
      _))
  (CallFunction|ShouldBounce? _impactnormal))

(fn OnRep_HitLocation)

(fn ShouldBounce?
  (params (Hit%20Normal :type Vector))
  ;[1 1 1] Only bounce if we are on a steep enough slope
  (if (> (Math|Vector|DotProduct Hit%20Normal (Math|Vector|GetUpVector "0, 0, 0")) 0.65)
    (Variables|/Script/Engine.ProjectileMovementComponent|SetbShouldBounce false (Variables|Movement|GetProjectileMovementComponent))))

(fn ShouldApplySnowman?
  (params (PlayerPawn :type Object%20Reference))
  (bind _returnvalue (GameplayTags|HasMatchingGameplayTag(Message) PlayerPawn "(TagName=\"Athena.Item.Snowman\")")) ; Make sure they aren't already a snowman
  (return (and (not _returnvalue) (Utilities|NotEqual(Object) PlayerPawn (Game|GetInstigator)))))

(fn UpdateQuest
  (params (FortPC :type Fort%20Player%20Controller%20Object%20Reference))
  ;[1 1 1] Quest hook (if this ever needs to change please let Dave.Ewing or Nicholas.Lacelle know)
  (Networking|SwitchHasAuthority
    (:Authority
      (Quest|SendComplexCustomStatEvent (Quests|GetQuestManager FortPC "Invalid") 0 "(GameplayTags=((TagName=\"Gameplay.Athena.Quests.UseSneakySnowman\")))" "(GameplayTags=)" 1))))

(fn ConstructionScript
  (Networking|CanExecuteCosmeticEvents
    (:True
      (Components|Movement|SetUpdatedComponent (Variables|Default|GetRotatingMovement) (Variables|Default|GetSK_Sneaky_Snowman_Throw))
      (Audio|SetCollisionSoundatIndex (Variables|Default|GetFortCollisionAudio) 0 (Variables|Sounds|GetCue_HitPlayer)))))
