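; Gameplay ability for the Athena Sneaky Snowman item that takes the snowman off a player.
; On activation it runs generic ability setup, stores the player pawn, and destroys attached Sneaky Snowman actors in RemoveSnowman, which also calls EndAbility.
; It then removes owner effects granting the Athena.Item.Snowman tag, sends the generic triggered-ability gameplay event to the pawn, and forces a net update.
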(event Ability|EventActivateAbility
  (bind _playerpawn (Variables|Default|GetPlayerPawn))
  (|CallParentFunction)
  (bind (_event_data _fort_player_pawn) (Setup|SetupGenericAbility false false)
    (:Output
      (bind _output_get (Variables|Default|SetPlayerPawn _fort_player_pawn))
      (CallFunction|RemoveSnowman _output_get)
      (Ability|RemoveGameplayEffectFromOwnerWithGrantedTags "(GameplayTags=((TagName=\"Athena.Item.Snowman\")))")
      (Ability|SendGameplayEventtoActor _playerpawn "(TagName=\"Abilities.Generic.TriggeredAbility\")")
      (Networking|ForceNetUpdate _playerpawn))))

(fn RemoveSnowman
  (params (PlayerPawn :type Fort%20Player%20Pawn%20Object%20Reference))
  (for _array_element (Actor|GetAttachedActors PlayerPawn true false)
    (bind _asathena_enemy_player_sneaky_snowman (Utilities|Casting|CastToAthena_Enemy_Player_SneakySnowman _array_element))
    (Utilities|IsValid _asathena_enemy_player_sneaky_snowman
      (:Is%20Valid
        (Actor|DestroyActor _asathena_enemy_player_sneaky_snowman))))
  (Ability|EndAbility))
