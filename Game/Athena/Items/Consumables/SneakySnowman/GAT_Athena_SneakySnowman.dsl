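; Gameplay ability that takes the player out of the Sneaky Snowman disguise and cleans up the consumable.
; On activation it gets the pawn and controller, then removes zero-stack snowman item instances, with a fallback removal call if the item definition is invalid.
; It removes effects granting the Athena.Item.Snowman tag and forces a net update.
; RemoveSnowman then destroys attached Athena_Player_SneakySnowman actors, covering the damage case, and ends the ability.
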(event Ability|EventActivateAbility
  (bind _playerpawn (Variables|Default|GetPlayerPawn))
  (|CallParentFunction)
  (bind (_event_data _fort_player_pawn) (Setup|SetupGenericAbility false false)
    (:Output
      (Variables|Default|SetPlayerPawn _fort_player_pawn)
      ;[1 1 1] Remove Item from player if they don't have any charges left
      (bind _asfort_player_controller (Utilities|Casting|CastToFortPlayerController (Pawn|GetController (Variables|Default|GetPlayerPawn)))
        (:then
          (Utilities|IsValid (Variables|Default|GetSnowmanItemDef)
            (:Is%20Valid
              (Inventory|FindItemInstancesFromDefinition :self _asfort_player_controller :ItemDefinition (Variables|Default|GetSnowmanItemDef) :ItemArray (Variables|Default|GetSnowmanItemInstances))
              (for _array_element (Variables|Default|GetSnowmanItemInstances)
                (if (== (Item|GetNuminStack _array_element) 0)
                  (Inventory|RemoveFortItemFromPlayer _asfort_player_controller _array_element -1 true)))
              (Utilities|Array|Clear (Variables|Default|GetSnowmanItemInstances)))
            (:Is%20Not%20Valid
              (Inventory|RemoveItemFromPlayer :PlayerController _asfort_player_controller :ItemDefinition (Variables|Default|GetSnowmanItemDef) :AmountToRemove -1 :bForceRemoval true :bSkipVerbRemoval false)))
          (Ability|RemoveGameplayEffectFromOwnerWithGrantedTags "(GameplayTags=((TagName=\"Athena.Item.Snowman\")))")
          (Networking|ForceNetUpdate _playerpawn)
          ;[1 1 1] Remove the Snowman in case they took damage
          (CallFunction|RemoveSnowman _playerpawn))
        (:CastFailed
          _)))))

(fn RemoveSnowman
  (params (PlayerPawn :type Fort%20Player%20Pawn%20Object%20Reference))
  (for _array_element (Actor|GetAttachedActors PlayerPawn true false)
    (bind _asathena_player_sneaky_snowman (Utilities|Casting|CastToAthena_Player_SneakySnowman _array_element))
    (Utilities|IsValid _asathena_player_sneaky_snowman
      (:Is%20Valid
        (Actor|DestroyActor _asathena_player_sneaky_snowman))))
  (Ability|EndAbility))
