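# Subsurface material for the Cool Carpet item in the Sneaky Snowman consumable folder, with a timed spawn-style wobble.
# World position offset scales vertices out from the actor pivot using eased time since WobbleStartTime, adding a decaying per-axis oscillation whose timing and strength vary radially via a sphere mask.
# Textures drive surface attributes, with Specular reusing the metallic channel, and SSS-tinted diffuse glows on downward-facing normals.
# The attributes pass through MF_CreativeMoveTool.
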Material /Game/Athena/Items/Consumables/SneakySnowman/Materials/M_CoolCarpet_01
graph {
 properties {
  bUseMaterialAttributes=true
  bUsedWithMeshDeformer=true
  bUsedWithSkeletalMesh=true
  pixelDepthOffsetMode=PDOM_Legacy
  shadingModel=MSM_Subsurface
  usedShadingModels=Subsurface
 }
 nodes [
  n1 = TextureSampleParameter2D() {parameterName=Diffuse, texture=/Game/Gadgets/Assets/CoolCarpet/Textures/T_Item_Cool_Carpet_D}
  n2 = TextureSampleParameter2D() {
   desc="R = SpecMask G = Metallic B = Roughness"
   group=Specular
   parameterName=SpecularMasks
   samplerSource=SSM_Wrap_WorldGroupSettings
   samplerType=SAMPLERTYPE_Masks
   texture=/Game/Gadgets/Assets/CoolCarpet/Textures/T_Item_Cool_Carpet_S
  }
  n3 = SphereMask(A=ComponentMask(None=WorldPosition().XYZ) {g=true, r=true}, B=ComponentMask(None=ObjectPositionWS()) {g=true, r=true}, Radius=ObjectRadius()) {hardnessPercent=0}
  n4 = LinearInterpolate(Alpha=@n3) {constA=0.7}
  n5 = WorldPosition()
  n6 = ActorPositionWS()
  n7 = Time() {period=5}
  n8 = ScalarParameter() {parameterName=WobbleStartTime}
  n9 = Clamp(None=Divide(A=Subtract(A=@n7, B=@n8), B=ScalarParameter() {defaultValue=1, group=Timestamp, parameterName=WobbleDuration}))
  n10 = Clamp(None=Power(Base=@n9, Exp=ScalarParameter() {defaultValue=0.2, parameterName=EaseOutPower}) {constExponent=0.2})
  n11 = Add(
   A=Multiply(A=OneMinus(None=@n3), B=ScalarParameter() {defaultValue=-0.05, parameterName=OuterWobbleTimeOffset}) {constB=0.1}
   B=Power(Base=@n9) {constExponent=1.5}
  )
  n12 = LinearInterpolate(
   A=VectorParameter() {parameterName=WobbleMult1, defaultValue=<rgba 1 1 0 1>}.RGB
   B=VectorParameter() {parameterName=WobbleMult2, defaultValue=<rgba 0 0 2 1>}.RGB
   Alpha=ConstantBiasScale(None=Sine(None=@n11) {period=0.25})
  )
  n13 = Sine(None=@n10)
  n14 = LinearInterpolate(Alpha=ConstantBiasScale(None=OneMinus(None=@n13))) {constA=1.25, constB=0}
  n15 = Multiply(
   A=Multiply(A=Subtract(A=@n5.XYZ, B=@n6), B=LinearInterpolate(Alpha=@n10) {constA=-1, constB=0})
   B=Add(A=VectorParameter() {parameterName=ScaleAxis, defaultValue=<rgba 0.5 0.5 1 1>}.RGB, B=Multiply(A=@n12, B=@n14))
  )
  n16 = VectorParameter() {parameterName=SSS}
  n17 = Saturate(None=DotProduct(A=VertexNormalWS(), B=Constant3Vector() {constant=<rgba 0 0 -1 0>}))
  n18 = SetMaterialAttributes!(
   MaterialAttributes
   Base Color=@n1.RGB
   Metallic=@n2.G
   Specular=@n2.G
   Roughness=LinearInterpolate(
    A=ScalarParameter() {group=Roughness, parameterName=SpecRoughnessMin}
    B=ScalarParameter() {defaultValue=1, group=Roughness, parameterName=SpecRoughnessMax}
    Alpha=@n2.B
   )
   Normal=TextureSampleParameter2D() {parameterName=Normal, samplerType=SAMPLERTYPE_Normal, texture=/Game/Gadgets/Assets/CoolCarpet/Textures/T_Item_Cool_Carpet_N}.RGB
   World Position Offset=Multiply(A=Multiply(A=@n4, B=@n15) {constA=1}, B=ScalarParameter() {defaultValue=1, parameterName=WobbleStrength})
   Subsurface Color=Multiply(A=@n1.RGB, B=@n16.RGB)
   Emissive Color=Multiply(A=@n1.RGB, B=Multiply(A=@n17, B=@n16.RGB))
  ) {attributeSetTypes [
    #1
    #2
    #3
    #4
    #5
    #6
    #7
    #8
   ]}
  MaterialFunctionCall(MA In=@n18) {materialFunction=&MF_CreativeMoveTool}
  Time() {bOverride_Period=true, period=1}
 ]
}
