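# M_DepthShadow is an unlit, translucent, two-sided material for the Sneaky Snowman consumable, named as a depth-based shadow.
# Its graph holds a zero-valued Constant and a DepthFade node driven by Opacity and FadeDistance scalar parameters.
# DepthFade softens the opacity value where the surface nears opaque geometry behind it, with both strength and fade distance tunable in material instances.
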Material /Game/Athena/Items/Consumables/SneakySnowman/Materials/M_DepthShadow
graph {
 properties {blendMode=BLEND_Translucent, pixelDepthOffsetMode=PDOM_Legacy, shadingModel=MSM_Unlit, twoSided=true, usedShadingModels=Unlit}
 nodes [
  Constant()
  DepthFade(
   Opacity=ScalarParameter() {defaultValue=0.85, parameterName=Opacity}
   FadeDistance=ScalarParameter() {defaultValue=20, parameterName=FadeDistance}
  )
 ]
}
